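# Niagara idle effect for the SpecialScroll Geisha backpack skin, adding pale blue glows to eyes and horns plus a soft light.
# EyesGlow and HornsGlow burst persistent sprites, using partition indices to set each one's bone, offset and size on the skeletal mesh.
# Glow brightness switches between gameplay and front end, where System.FadeIn scales it; camera-angle alpha fades are disabled.
# The Light emitter's offset light is tinted by System.TOD.
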NiagaraSystem /BRCosmetics/Accessories/FORT_Backpacks/Backpack_M_MED_Geisha/Skins/SpecialScroll/FX/NS_SpecialScroll_Backpack_Idle
dependencies {
 usedDataInterfaces [&NiagaraDataInterfaceCamera, &NiagaraDataInterfaceSkeletalMesh, &NiagaraDataInterfaceCurve]
 usedDynamicInputs [
  &FloatFromCurve
  &Subtract_Float
  /Game/Effects/Niagara/DynamicInputs/TimeOfDay/DaySequenceTimeOfDayColor
 ]
 usedModules [
  &self:SystemSpawnScript.SystemScriptSource.SystemScriptGraph.NiagaraNodeAssignment_0.SetVariables_#1
  &SystemState
  &self:SystemSpawnScript.SystemScriptSource.SystemScriptGraph.NiagaraNodeAssignment_2.SetVariables_#2
  &self:SystemSpawnScript.SystemScriptSource.SystemScriptGraph.NiagaraNodeAssignment_3.SetVariables_#3
 ]
}
emitters {
 EyesGlow {
  emitterUpdateScript [
   &EmitterState(Life Cycle Mode=::NewEnumerator1, Loop Behavior=::NewEnumerator1, Loop Duration Mode=::NewEnumerator1, UseLoopDelay=false)
   &SpawnBurst_Instantaneous(Spawn Count=2, Use Spawn Probability=false, UseLoopCountLimit=false)
  ]
  particleSpawnScript [
   &PartitionParticles()
   &InitializeParticle(
    Color=<rgba 13.4115 19.7143 25 0.05>
    Sprite Size Mode=::NewEnumerator3
    Uniform Sprite Size=&Multiply_Float(
     A=&MergeHelper_FloatSwitch(0=22, 1=22, 2=?, 3=?, Index=$Particles.PartitionParticles.Partition, Num=::NewEnumerator0)
     B=1
    )
    UsePositionOffset=false
    Write Material Random=true
   )
  ]
  particleUpdateScript [
   &ParticleState(Kill Particles When Lifetime Has Elapsed=false, Let Infinitely Lived Particles Die When Emitter Deactivates=true)
   &SkeletalMeshLocation(
    Bone / Socket Index=&MergeHelper_IntSwitch(0=5, 1=9, 2=?, 3=?, 4=?, 5=?, Index=$Particles.PartitionParticles.Partition, Num=::NewEnumerator0)
    Bone Sampling Mode=::NewEnumerator6
    MaskSpawnGroup=false
    Sampled Position Offset=&AddVectorToPosition(
     Position=<0 0 0>
     Vector=&MergeHelper_VectorSwitch(0=<-1 0 0>, 1=<1 0 0>, 2=?, 3=?, Index=$Particles.PartitionParticles.Partition, Num=::NewEnumerator0)
    )
    Skeletal Mesh=$System.Skeletal Mesh
   )
   &ScaleColor(
    Scale Alpha=&Multiply_Float(
     A=&Multiply_Float(A=1, B=1)
     B=&ClampFloat(
      Float=&Power(
       a=&Abs_Float(Float:NiagaraFloat=&DotProduct(Vector A=&ReturnCameraVector(Camera Vectors=::NewEnumerator1, Coordinate Space=::Local), Vector B=<0 1 0>))
       b=8
      )
      Min=&Multiply_Float(A=1, B=&FN_FrontEndFloatModifier(Gameplay=1, Front End=1))
      Max=&Multiply_Float(A=1, B=&FN_FrontEndFloatModifier(Gameplay=0.136, Front End=0.2))
     )
    )
    ScaleA=true
    ScaleRGB=false
   ) {enabled=false}
   &ScaleColor(
    Scale Alpha=&FN_FrontEndFloatModifier(Gameplay=0.1, Front End=&Multiply_Float(A=0.05, B=$System.FadeIn))
    ScaleA=true
    ScaleRGB=true
   ) {module=ScaleColor001}
  ]
  rendererClasses [&NiagaraSpriteRendererProperties]
  renderers [{rendererClass=&NiagaraSpriteRendererProperties}]
 }
 HornsGlow {
  emitterUpdateScript [
   &EmitterState(Life Cycle Mode=::NewEnumerator1, Loop Behavior=::NewEnumerator1, Loop Duration Mode=::NewEnumerator1, UseLoopDelay=false)
   &SpawnBurst_Instantaneous(Spawn Count=8, Use Spawn Probability=false, UseLoopCountLimit=false)
  ]
  particleSpawnScript [
   &PartitionParticles(Number Of Partitions=8)
   &InitializeParticle(
    Color=<rgba 13.4115 19.7143 25 0.05>
    Sprite Size Mode=::NewEnumerator3
    Uniform Sprite Size=&Multiply_Float(
     A=&MergeHelper_FloatSwitch(0=13, 1=22, 2=13, 3=22, 4=15, 5=15, 6=8, 7=8, Index=$Particles.PartitionParticles.Partition, Num=::NewEnumerator6)
     B=1
    )
    UsePositionOffset=false
    Write Material Random=true
   )
  ]
  particleUpdateScript [
   &ParticleState(Kill Particles When Lifetime Has Elapsed=false, Let Infinitely Lived Particles Die When Emitter Deactivates=true)
   &SkeletalMeshLocation(
    Bone / Socket Index=&MergeHelper_IntSwitch(0=2, 1=2, 2=6, 3=6, 4=4, 5=8, 6=4, 7=8, Index=$Particles.PartitionParticles.Partition, Num=::NewEnumerator6)
    Bone Sampling Mode=::NewEnumerator6
    MaskSpawnGroup=false
    Sampled Position Offset=&AddVectorToPosition(
     Position=<0 0 0>
     Vector=&MergeHelper_VectorSwitch(
      0=<-3 -3 5>
      1=<-1.5 -1 2>
      2=<3 3 -5>
      3=<1.5 1 -2>
      4=<0 0 0>
      5=<0 0 0>
      6=<-2 -0.5 -1.2>
      7=<2 0.5 1.2>
      Index=$Particles.PartitionParticles.Partition
      Num=::NewEnumerator6
     )
    )
    Skeletal Mesh=$System.Skeletal Mesh
   )
   &ScaleColor(
    Scale Alpha=&Multiply_Float(
     A=&Multiply_Float(A=1, B=1)
     B=&ClampFloat(
      Float=&Power(
       a=&Abs_Float(Float:NiagaraFloat=&DotProduct(Vector A=&ReturnCameraVector(Camera Vectors=::NewEnumerator1, Coordinate Space=::Local), Vector B=<0 1 0>))
       b=8
      )
      Min=&Multiply_Float(A=1, B=&FN_FrontEndFloatModifier(Gameplay=1, Front End=1))
      Max=&Multiply_Float(A=1, B=&FN_FrontEndFloatModifier(Gameplay=0.136, Front End=0.2))
     )
    )
    ScaleA=true
    ScaleRGB=false
   ) {enabled=false}
   &ScaleColor(
    Scale Alpha=&FN_FrontEndFloatModifier(Gameplay=0.1, Front End=&Multiply_Float(A=0.05, B=$System.FadeIn))
    ScaleA=true
    ScaleRGB=true
   ) {module=ScaleColor001}
  ]
  rendererClasses [&NiagaraSpriteRendererProperties]
  renderers [{rendererClass=&NiagaraSpriteRendererProperties}]
 }
 Light {
  emitterUpdateScript [
   &EmitterState(
    Inactive Response=::NewEnumerator1
    Life Cycle Mode=::NewEnumerator1
    Loop Behavior=::NewEnumerator1
    Loop Duration Mode=::NewEnumerator1
    UseLoopDelay=false
   )
   &SpawnBurst_Instantaneous(Use Spawn Probability=false, UseLoopCountLimit=false)
  ]
  particleSpawnScript [&InitializeParticle(
    Color=&MultiplyLinearColor(Linear Color=<rgba 0.7552 0.7552 0.7552 0.3>, Scale Factor=<1 1 1 1>)
    Lifetime=&RandomRangeFloat(Minimum=0.6, Maximum=1)
    Sprite Size=<8 70>
    Sprite Size Mode=::NewEnumerator4
    UsePositionOffset=false
    Write Material Random=true
   )]
  particleUpdateScript [
   &ParticleState(Kill Particles When Lifetime Has Elapsed=false)
   &FN_Light_Attributes(
    Color=&Multiply_LinearColorByFloat(
     Linear Color=&MultiplyLinearColor(Linear Color=<rgba 0.3307 0.4 0.5 1>, Scale Factor=&MakeVector4_FromLinearColor(LinearColor=$System.TOD))
     Float=1.2
    )
    Create Color Attribute=true
    Create Exponent Attribute=true
    Create Position Attribute=true
    Create Radius Attribute=true
    Exponent=2
    Position=&AddVectorToPosition(Position=$Particles.Position, Vector=<0 -15 0>)
    Radius=60
   )
  ]
  rendererClasses [&NiagaraLightRendererProperties]
  renderers [{rendererClass=&NiagaraLightRendererProperties, rendererIndex=0}]
 }
}
summary {
 systemName=NS_SpecialScroll_Backpack_Idle
 emitters [
  {emitterName=HornsGlow, rendererClasses [&NiagaraSpriteRendererProperties]}
  {emitterName=Light, rendererClasses [&NiagaraLightRendererProperties]}
  {emitterName=EyesGlow, rendererClasses [&NiagaraSpriteRendererProperties]}
 ]
 userVariables [
  User.bIsFrontEnd {type=NiagaraBool}
  User.bIsFrontEndPreview {type=NiagaraBool}
 ]
}
